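; Lobby countdown manager for a calendar event, driving the MPC_Countdown material parameter collection that displays the timer.
; When the watched event is active, a one-second timer splits days, hours, minutes and seconds into eight digit parameters via reversed strings, decrements the time and broadcasts OnSecondsRemainUpdated.
; A second timer re-queries the calendar every 30 seconds to correct drift.
; At zero or event end, digit parameters are zeroed with the countdown still shown.
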(event EventBeginPlay
  ;[1 1 1] Start timer if calendar is active
  (bind (_eventname _timeuntilend _timesincebegin _timespanratio) (MCPContext|WatchCalendarEvent (Variables|Default|GetCalendarEventName))
    (:OnEventActive
      (Variables|Default|SetTimeUntilCountdownEnd _timeuntilend)
      (Variables|Default|SetTimeSpanRatio _timespanratio)
      (Variables|Default|SetTimeSinceCountdownBegin _timesincebegin)
      (CallFunction|ContinueClockCountdown)
      ;[1 1 1] Every 30 seconds requery calendar time
      (bind _returnvalue (Utilities|Time|SetTimerbyEvent (EventDispatchers|CreateEvent :!event Custom|CountDownTimer_AccuracyFix) 3e+01 true))
      (Variables|Default|SetAccuracyTimer _returnvalue))
    (:OnEventEnded
      ;[1 1 1] Visual Update
      ;;[1 1 1] count down is over show final graphic
      (Rendering|Material|SetScalarParameterValue "/Game/Athena/Events/Generic/CountdownTimer/Materials/MPC_Countdown.MPC_Countdown" "ShowCountdown" 1.0 :!class /Script/Engine.KismetMaterialLibrary)
      (Rendering|Material|SetScalarParameterValue "/Game/Athena/Events/Generic/CountdownTimer/Materials/MPC_Countdown.MPC_Countdown" "ShowCountdownPhase2" 1.0 :!class /Script/Engine.KismetMaterialLibrary) ; Shows 2nd stage of countdown (hologram countdown floating in the sky)
      ;[1 1 1] 0 0 0 0 0 0 0 0 0
      (Rendering|Material|SetScalarParameterValue "/Game/Athena/Events/Generic/CountdownTimer/Materials/MPC_Countdown.MPC_Countdown" "Digit1" :!class /Script/Engine.KismetMaterialLibrary)
      (Rendering|Material|SetScalarParameterValue "/Game/Athena/Events/Generic/CountdownTimer/Materials/MPC_Countdown.MPC_Countdown" "Digit2" :!class /Script/Engine.KismetMaterialLibrary)
      (Rendering|Material|SetScalarParameterValue "/Game/Athena/Events/Generic/CountdownTimer/Materials/MPC_Countdown.MPC_Countdown" "Digit3" :!class /Script/Engine.KismetMaterialLibrary)
      (Rendering|Material|SetScalarParameterValue "/Game/Athena/Events/Generic/CountdownTimer/Materials/MPC_Countdown.MPC_Countdown" "Digit4" :!class /Script/Engine.KismetMaterialLibrary)
      (Rendering|Material|SetScalarParameterValue "/Game/Athena/Events/Generic/CountdownTimer/Materials/MPC_Countdown.MPC_Countdown" "Digit5" :!class /Script/Engine.KismetMaterialLibrary)
      (Rendering|Material|SetScalarParameterValue "/Game/Athena/Events/Generic/CountdownTimer/Materials/MPC_Countdown.MPC_Countdown" "Digit6" :!class /Script/Engine.KismetMaterialLibrary)
      (Rendering|Material|SetScalarParameterValue "/Game/Athena/Events/Generic/CountdownTimer/Materials/MPC_Countdown.MPC_Countdown" "Digit7" :!class /Script/Engine.KismetMaterialLibrary)
      (Rendering|Material|SetScalarParameterValue "/Game/Athena/Events/Generic/CountdownTimer/Materials/MPC_Countdown.MPC_Countdown" "Digit8" :!class /Script/Engine.KismetMaterialLibrary)
      ;[1 1 1] SetTimeSpan Ratio (normalized progress)
      (Rendering|Material|SetScalarParameterValue "/Game/Athena/Events/Generic/CountdownTimer/Materials/MPC_Countdown.MPC_Countdown" "CountdownNomalizedProgress" (Variables|Default|GetTimeSpanRatio) :!class /Script/Engine.KismetMaterialLibrary))) ; Leadup to Festivus
)

(event Custom|ContinueClockCountdown
  ;[1 1 1] Set seconds remaining and start visual updates every second
  (Variables|Default|SetTotalSecondsAtEvent (Math|Timespan|GetTotalSeconds (Variables|Default|GetTimeUntilCountdownEnd)))
  (if (not (Utilities|Time|IsTimerActivebyHandle (Variables|Default|GetCountdownTimerHandle)))
    (bind _returnvalue (Utilities|Time|SetTimerbyEvent (EventDispatchers|CreateEvent :!event Custom|CountdownTimer_New) 1.0 true))
    (Variables|Default|SetCountdownTimerHandle _returnvalue)))

(event Custom|CountdownTimer_New
  (bind _timeuntilcountdownend (Variables|Default|GetTimeUntilCountdownEnd))
  ;[1 1 1] count down is over show final graphic
  (Development|PrintString "Tick")
  (CallFunction|VisualUpdateEverySecond)
  (Default|CallOnSecondsRemainUpdated (Math|Float|Truncate (Math|Timespan|GetTotalSeconds _timeuntilcountdownend)))
  (if (<= (Math|Timespan|GetTotalSeconds _timeuntilcountdownend) 0.0)
    (Utilities|FlowControl|DoOnce false
      (:Completed
        (Utilities|Time|ClearandInvalidateTimerbyHandle (Variables|Default|GetCountdownTimerHandle))
        (Rendering|Material|SetScalarParameterValue "/Game/Athena/Events/Generic/CountdownTimer/Materials/MPC_Countdown.MPC_Countdown" "ShowCountdown" 1.0 :!class /Script/Engine.KismetMaterialLibrary)
        (Rendering|Material|SetScalarParameterValue "/Game/Athena/Events/Generic/CountdownTimer/Materials/MPC_Countdown.MPC_Countdown" "ShowCountdownPhase2" 1.0 :!class /Script/Engine.KismetMaterialLibrary) ; Shows 2nd stage of countdown (hologram countdown floating in the sky)
      ))))

(event Custom|ClearTimerVisuals)

(event Custom|CountDownTimer_AccuracyFix
  ;[1 1 1] Fix Accuracy every once in a wile based on timer ( more often in the last minutes)
  (bind (_biseventactive _timesincebegin _timeuntilend) (CalendarEvents|GetCalendarEventInformation (Variables|Default|GetCalendarEventName)))
  (if _biseventactive
    (Variables|Default|SetTimeUntilCountdownEnd _timeuntilend)
    (Variables|Default|SetTotalSecondsAtEvent (Math|Timespan|GetTotalSeconds (Variables|Default|GetTimeUntilCountdownEnd)))
    (else
      (Utilities|Time|ClearandInvalidateTimerbyHandle (Variables|Default|GetAccuracyTimer)))))

(event Custom|VisualUpdateEverySecond
  (bind _returnvalue (Math|Timespan|GetSeconds (Variables|Default|GetTimeUntilCountdownEnd)))
  (bind _returnvalue_2 (Utilities|String|ToString(Integer) _returnvalue))
  (bind _returnvalue_3 (Math|Timespan|GetHours (Variables|Default|GetTimeUntilCountdownEnd)))
  (bind _returnvalue_4 (Utilities|String|ToString(Integer) _returnvalue_3))
  (bind _returnvalue_5 (Math|Timespan|GetDays (Variables|Default|GetTimeUntilCountdownEnd)))
  (bind _returnvalue_6 (Utilities|String|ToString(Integer) _returnvalue_5))
  (bind _returnvalue_7 (Utilities|String|Reverse _returnvalue_6))
  (bind _returnvalue_8 (Math|Timespan|GetMinutes (Variables|Default|GetTimeUntilCountdownEnd)))
  (bind _returnvalue_9 (Utilities|String|ToString(Integer) _returnvalue_8))
  (bind _returnvalue_10 (Utilities|String|Reverse _returnvalue_9))
  (bind _returnvalue_11 (Utilities|String|Reverse _returnvalue_2))
  (bind _returnvalue_12 (Utilities|String|Reverse _returnvalue_4))
  (bind _timeuntilcountdownend (Variables|Default|GetTimeUntilCountdownEnd))
  ;[1 1 1] Visual Update
  (if (<= (Math|Timespan|GetTotalSeconds (Variables|Default|GetTimeUntilCountdownEnd)) 0.0)
    ;;[1 1 1] count down is over show final graphic
    (Rendering|Material|SetScalarParameterValue "/Game/Athena/Events/Generic/CountdownTimer/Materials/MPC_Countdown.MPC_Countdown" "ShowCountdown" 1.0 :!class /Script/Engine.KismetMaterialLibrary)
    (Rendering|Material|SetScalarParameterValue "/Game/Athena/Events/Generic/CountdownTimer/Materials/MPC_Countdown.MPC_Countdown" "ShowCountdownPhase2" 1.0 :!class /Script/Engine.KismetMaterialLibrary) ; Shows 2nd stage of countdown (hologram countdown floating in the sky)
    ;[1 1 1] 0 0 0 0 0 0 0 0 0
    (Rendering|Material|SetScalarParameterValue "/Game/Athena/Events/Generic/CountdownTimer/Materials/MPC_Countdown.MPC_Countdown" "Digit1" :!class /Script/Engine.KismetMaterialLibrary)
    (Rendering|Material|SetScalarParameterValue "/Game/Athena/Events/Generic/CountdownTimer/Materials/MPC_Countdown.MPC_Countdown" "Digit2" :!class /Script/Engine.KismetMaterialLibrary)
    (Rendering|Material|SetScalarParameterValue "/Game/Athena/Events/Generic/CountdownTimer/Materials/MPC_Countdown.MPC_Countdown" "Digit3" :!class /Script/Engine.KismetMaterialLibrary)
    (Rendering|Material|SetScalarParameterValue "/Game/Athena/Events/Generic/CountdownTimer/Materials/MPC_Countdown.MPC_Countdown" "Digit4" :!class /Script/Engine.KismetMaterialLibrary)
    (Rendering|Material|SetScalarParameterValue "/Game/Athena/Events/Generic/CountdownTimer/Materials/MPC_Countdown.MPC_Countdown" "Digit5" :!class /Script/Engine.KismetMaterialLibrary)
    (Rendering|Material|SetScalarParameterValue "/Game/Athena/Events/Generic/CountdownTimer/Materials/MPC_Countdown.MPC_Countdown" "Digit6" :!class /Script/Engine.KismetMaterialLibrary)
    (Rendering|Material|SetScalarParameterValue "/Game/Athena/Events/Generic/CountdownTimer/Materials/MPC_Countdown.MPC_Countdown" "Digit7" :!class /Script/Engine.KismetMaterialLibrary)
    (Rendering|Material|SetScalarParameterValue "/Game/Athena/Events/Generic/CountdownTimer/Materials/MPC_Countdown.MPC_Countdown" "Digit8" :!class /Script/Engine.KismetMaterialLibrary)
    ;[1 1 1] SetTimeSpan Ratio (normalized progress)
    (Rendering|Material|SetScalarParameterValue "/Game/Athena/Events/Generic/CountdownTimer/Materials/MPC_Countdown.MPC_Countdown" "CountdownNomalizedProgress" (Variables|Default|GetTimeSpanRatio) :!class /Script/Engine.KismetMaterialLibrary)
    (else
      (Rendering|Material|SetScalarParameterValue "/Game/Athena/Events/Generic/CountdownTimer/Materials/MPC_Countdown.MPC_Countdown" "ShowCountdownPhase2" 1.0 :!class /Script/Engine.KismetMaterialLibrary)
      (Rendering|Material|SetScalarParameterValue "/Game/Athena/Events/Generic/CountdownTimer/Materials/MPC_Countdown.MPC_Countdown" "ShowCountdownPhase3" :!class /Script/Engine.KismetMaterialLibrary)
      (Rendering|Material|SetScalarParameterValue "/Game/Athena/Events/Generic/CountdownTimer/Materials/MPC_Countdown.MPC_Countdown" "ShowCountdown" 1.0 :!class /Script/Engine.KismetMaterialLibrary)
      (Rendering|Material|SetScalarParameterValue "/Game/Athena/Events/Generic/CountdownTimer/Materials/MPC_Countdown.MPC_Countdown" "Digit1" (Utilities|String|StringToFloat (Utilities|String|GetSubstring _returnvalue_11)) :!class /Script/Engine.KismetMaterialLibrary)
      (Rendering|Material|SetScalarParameterValue "/Game/Athena/Events/Generic/CountdownTimer/Materials/MPC_Countdown.MPC_Countdown" "Digit2" (Utilities|String|StringToFloat (Utilities|String|GetSubstring _returnvalue_11 1)) :!class /Script/Engine.KismetMaterialLibrary)
      (Rendering|Material|SetScalarParameterValue "/Game/Athena/Events/Generic/CountdownTimer/Materials/MPC_Countdown.MPC_Countdown" "Digit3" (Utilities|String|StringToFloat (Utilities|String|GetSubstring _returnvalue_10)) :!class /Script/Engine.KismetMaterialLibrary)
      (Rendering|Material|SetScalarParameterValue "/Game/Athena/Events/Generic/CountdownTimer/Materials/MPC_Countdown.MPC_Countdown" "Digit4" (Utilities|String|StringToFloat (Utilities|String|GetSubstring _returnvalue_10 1)) :!class /Script/Engine.KismetMaterialLibrary)
      (Rendering|Material|SetScalarParameterValue "/Game/Athena/Events/Generic/CountdownTimer/Materials/MPC_Countdown.MPC_Countdown" "Digit5" (Utilities|String|StringToFloat (Utilities|String|GetSubstring _returnvalue_12)) :!class /Script/Engine.KismetMaterialLibrary)
      (Rendering|Material|SetScalarParameterValue "/Game/Athena/Events/Generic/CountdownTimer/Materials/MPC_Countdown.MPC_Countdown" "Digit6" (Utilities|String|StringToFloat (Utilities|String|GetSubstring _returnvalue_12 1)) :!class /Script/Engine.KismetMaterialLibrary)
      (Rendering|Material|SetScalarParameterValue "/Game/Athena/Events/Generic/CountdownTimer/Materials/MPC_Countdown.MPC_Countdown" "Digit7" (Utilities|String|StringToFloat (Utilities|String|GetSubstring _returnvalue_7)) :!class /Script/Engine.KismetMaterialLibrary)
      (Rendering|Material|SetScalarParameterValue "/Game/Athena/Events/Generic/CountdownTimer/Materials/MPC_Countdown.MPC_Countdown" "Digit8" (Utilities|String|StringToFloat (Utilities|String|GetSubstring _returnvalue_7 1)) :!class /Script/Engine.KismetMaterialLibrary)
      (Variables|Default|SetTimeUntilCountdownEnd (Math|Timespan|MakeTimespan (Math|Timespan|GetDays (Variables|Default|GetTimeUntilCountdownEnd)) (Math|Timespan|GetHours (Variables|Default|GetTimeUntilCountdownEnd)) (Math|Timespan|GetMinutes (Variables|Default|GetTimeUntilCountdownEnd)) (- (Math|Timespan|GetSeconds (Variables|Default|GetTimeUntilCountdownEnd)) 1)))
      ;[1 1 1] SetTimeSpan Ratio (normalized progress)
      (Rendering|Material|SetScalarParameterValue "/Game/Athena/Events/Generic/CountdownTimer/Materials/MPC_Countdown.MPC_Countdown" "CountdownNomalizedProgress" (Variables|Default|GetTimeSpanRatio) :!class /Script/Engine.KismetMaterialLibrary))))

(fn ConstructionScript)
